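; DialogVM_PurchaseConfirmation is a view model supplying message text for the Item Shop purchase confirmation dialog.
; SetOfferVM takes a FortItemShopOfferVM and copies its matchmaking-ban and generic purchase confirmation messages into the MatchmakingBanInfo and PurchaseConfirmation variables.
; When the offer is not valid, both variables are set to empty.
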(fn SetOfferVM
  (params (InOfferVM :type Fort%20Item%20Shop%20Offer%20VM%20Object%20Reference))
  (Utilities|IsValid InOfferVM
    (:Is%20Valid
      (Variables|PurchaseConfirmation|SetMatchmakingBanInfo (Class|FortItemShopOfferVM|GetPurchaseConfirmationMessageMatchmakingBan (Variables|GetInOfferVM)))
      (Variables|PurchaseConfirmation|SetPurchaseConfirmation (Class|FortItemShopOfferVM|GetPurchaseConfirmationMessageGeneric (Variables|GetInOfferVM))))
    (:Is%20Not%20Valid
      (Variables|PurchaseConfirmation|SetMatchmakingBanInfo "")
      (Variables|PurchaseConfirmation|SetPurchaseConfirmation ""))))
